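# Unlit prototype test material, apparently for a glowing beam in a tunnel environment.
# A pink-magenta Glow vector parameter is multiplied by a constant to boost its brightness.
# A SimpleGrassWind function, fed constant intensity, weight, speed and additional offset, computes wind sway of the kind used for world position offset.
# The graph lists no links from either branch to the material outputs.
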Material /Hera/Environment/Prototype/ChrisK/Tunnel/BeamTestMtl
graph {
 properties {bAutomaticallySetUsageInEditor=false, shadingModel=MSM_Unlit, usedShadingModels=Unlit}
 nodes [
  MaterialFunctionCall!(WindIntensity=Constant() {r=1}, WindWeight=Constant() {r=0.7}, WindSpeed=Constant() {r=0.7}, AdditionalWPO=Constant() {r=0.7}) {materialFunction=/Engine/Functions/Engine_MaterialFunctions01/WorldPositionOffset/SimpleGrassWind}
  Multiply(A=VectorParameter() {parameterName=Glow, defaultValue=<rgba 1 0.2865 0.9341 1>}.RGB) {constB=5}
 ]
}
