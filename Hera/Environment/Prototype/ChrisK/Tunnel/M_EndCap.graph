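# Translucent, unlit, two-sided glow material for a prototype tunnel end cap, with a static switch for a sprite variant.
# Sprite opacity multiplies a generated band by a V-axis falloff pulsed by a debug time sine; otherwise a radial sphere mask is used.
# Depth fade softens opacity at intersections.
# Emissive color hue-shifts across U with inverse eye adaptation, and the non-sprite variant adds a powered radial core.
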Material /Hera/Environment/Prototype/ChrisK/Tunnel/M_EndCap
graph {
 properties {bAutomaticallySetUsageInEditor=false, blendMode=BLEND_Translucent, shadingModel=MSM_Unlit, twoSided=true, usedShadingModels=Unlit}
 nodes [
  n1 = MaterialFunctionCall!(
   Width=ScalarParameter() {defaultValue=0.5, parameterName=Width}
   Sharpness
   Offset
   Direction Switch=StaticBool() {value=true}
   Compare
   Input Coordinates
  ) {materialFunction=/Engine/Functions/Engine_MaterialFunctions03/Procedurals/GeneratedBand}
  n2 = ComponentMask(None=TextureCoordinate()) {g=true}
  n3 = MaterialFunctionCall() {materialFunction=&DebugTimeSine}
  n4 = Multiply(
   A=Power(Base=OneMinus(None=@n2), Exp=LinearInterpolate(Alpha=@n3.Result) {constA=2, constB=5})
   B=Power(Base=@n2, Exp=ScalarParameter() {defaultValue=1, parameterName=TipPower}) {constExponent=1}
  )
  n5 = TextureCoordinate()
  n6 = Constant() {r=0.5}
  n7 = ScalarParameter() {defaultValue=0.495, parameterName=Radius}
  n8 = ScalarParameter() {parameterName=Hardness}
  DepthFade(
   Opacity=StaticSwitchParameter(True=Multiply(A=@n1.Result, B=@n4), False=SphereMask(A=@n5, B=@n6, Radius=@n7, Hardness=@n8) {hardnessPercent=0}) {parameterName=IsEndCapSprites}
   FadeDistance=ScalarParameter() {defaultValue=5, parameterName=DepthFadeDistance}
  )
  n9 = Multiply(A=ComponentMask(None=TextureCoordinate()) {r=true}, B=ScalarParameter() {defaultValue=0.1, parameterName=HueShiftAmount})
  n10 = VectorParameter() {parameterName=EmissiveColor, defaultValue=<rgba 0 0.4707 5 1>}
  n11 = Multiply(
   A=EyeAdaptationInverse(
    LightValueInput=MaterialFunctionCall(Hue Shift Percentage=@n9, Texture=@n10.RGB) {materialFunction=&HueShift}.Result
    AlphaInput=ScalarParameter() {defaultValue=1, parameterName=EyeAdaptionInverse}
   ).EyeAdaptationInverse
   B=ScalarParameter() {defaultValue=1, parameterName=EmissiveMult}
  )
  n12 = Power(
   Base=SphereMask(
    A=TextureCoordinate()
    B=Constant() {r=0.5}
    Radius=ScalarParameter() {defaultValue=1, parameterName=Radius2}
    Hardness=ScalarParameter() {parameterName=Hardness2}
   ) {hardnessPercent=0}
   Exp=ScalarParameter() {defaultValue=5, parameterName=CorePower}
  )
  n13 = ScalarParameter() {defaultValue=1, parameterName=CoreEmissiveMult}
  StaticSwitchParameter(True=@n11, False=Add(A=Multiply(A=@n12, B=@n13), B=@n11)) {parameterName=IsEndCapSprites}
 ]
}
