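# M_Glass_CKTest is a translucent, unlit test glass material in a prototype Tunnel environment folder.
# A Color vector parameter, defaulting to dark teal, feeds one multiply with no second input connected.
# A separate multiply scales a sharpened Fresnel term down to a faint edge-weighted value.
# The file does not show which material outputs these two results drive.
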Material /Hera/Environment/Prototype/ChrisK/Tunnel/M_Glass_CKTest
graph {
 properties {bAutomaticallySetUsageInEditor=false, blendMode=BLEND_Translucent, shadingModel=MSM_Unlit, usedShadingModels=Unlit}
 nodes [
  Multiply(A=VectorParameter() {parameterName=Color, defaultValue=<rgba 0 0.1798 0.2448 1>}.RGB)
  Multiply(A=Fresnel() {exponent=4}, B=Constant() {r=0.2})
 ]
}
